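# Masked two-sided material for holiday tree decorations, combining hanging ornaments and light garlands, blended by a mask from UVs and vertex alpha.
# Ornaments use subsurface shading with textured emissive, boosted for ray tracing by a Lumen multiplier.
# Garland bulbs step through six hues over time, offset by vertex blue, with emissive shaped by lamp mask and inverted Fresnel.
# Deferred adds destructible-damage vertex animation; mobile skips it, clamps minimum roughness and uses separate bulb brightness.
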Material /Game/Environments/Asteria/Props/Holiday/Holiday_Tree_A/Materials/M_Holiday_Tree_Lights
graph {
 properties {
  bUseMaterialAttributes=true
  bUsedWithInstancedStaticMeshes=true
  bUsedWithNanite=true
  blendMode=BLEND_Masked
  pixelDepthOffsetMode=PDOM_Legacy
  shadingModel=MSM_FromMaterialExpression
  twoSided=true
  usedShadingModels=Default Lit | Subsurface
 }
 nodes [
  n1 = ComponentMask(None=TextureCoordinate()) {g=true}
  n2 = Floor(None=LinearInterpolate(Alpha=@n1) {constA=-0.001, constB=3.3344})
  Abs(None=@n2)
  Constant3Vector() {constant=<rgba 0.01 0.01 0.01 0>}
  MaterialFunctionCall(Input=@n1, Input Low=Constant() {r=0.001}, Input High=Constant() {r=0.3}) {materialFunction=&RemapValueRange}
  n3 = Saturate(None=@n2)
  NamedRerouteDeclaration(None=Min(A=VertexColor().A, B=@n3)) {name=GarlandMask, nodeColor=<rgba 0 0.25 1 1>}
  n4 = TextureSample() {samplerSource=SSM_Wrap_WorldGroupSettings, texture=/Game/Environments/Props/Winter/Textures/T_ChristmasDeco_D}
  n5 = TextureSample() {samplerSource=SSM_Wrap_WorldGroupSettings, samplerType=SAMPLERTYPE_Masks, texture=/Game/Environments/Props/Winter/Textures/T_ChristmasDeco_S}
  n6 = TextureSample() {samplerSource=SSM_Wrap_WorldGroupSettings, texture=/Game/Environments/Props/Winter/Textures/T_ChristmasDeco_E}
  n7 = Multiply(A=Multiply(A=@n6.R, B=@n6.R), B=Constant3Vector() {constant=<rgba 0.1936 0.6163 1 1>})
  n8 = ScalarParameter() {defaultValue=3, parameterName=LumenMultiplier}
  n9 = TextureSample() {
   samplerSource=SSM_Wrap_WorldGroupSettings
   samplerType=SAMPLERTYPE_Normal
   texture=/Game/Environments/Props/Winter/Textures/T_ChristmasDeco_N
  }
  n10 = TextureSample() {samplerSource=SSM_Wrap_WorldGroupSettings, samplerType=SAMPLERTYPE_Masks, texture=/Game/Environments/Props/Winter/Textures/T_ChristmasDeco_M}
  n11 = SetMaterialAttributes!(
   MaterialAttributes
   Base Color=@n4.RGB
   Metallic=@n5.G
   Specular=@n5.G
   Roughness=@n5.G
   Emissive Color=RayTracingQualitySwitch(Normal=@n7, RayTraced=Multiply(A=@n7, B=@n8))
   Normal=@n9.RGB
   Opacity Mask=@n10.R
   Subsurface Color=Multiply(A=ScalarParameter() {defaultValue=0.2, parameterName=HangingOrnaments_SubsurfaceBrightness}, B=Multiply(A=@n10.R, B=@n4.RGB))
   Opacity=Multiply(A=OneMinus(None=@n3), B=Constant() {r=0.05})
   Shading Model=ShadingModel() {shadingModel=MSM_Subsurface}
  ) {attributeSetTypes [
    #1
    #2
    #3
    #4
    #5
    #6
    #7
    #8
    #9
    #10
   ]}
  n12 = TextureSample() {samplerType=SAMPLERTYPE_Masks, texture=/Game/Environments/Asteria/Props/Holiday/KIT_Holiday_Garlands_A/Testures/T_Holiday_Garlands_A_A_S}
  n13 = TextureSampleParameter2D() {
   parameterName=LampMask
   samplerType=SAMPLERTYPE_LinearColor
   texture=/Game/Environments/Asteria/Props/Holiday/KIT_Holiday_Garlands_A/Testures/T_Holiday_Garlands_A_A_E
  }
  n14 = Ceil(None=@n13.R)
  n15 = OneMinus(None=Fresnel() {baseReflectFraction=0})
  n16 = VertexColor()
  n17 = Time()
  n18 = Multiply(A=Floor(None=Multiply(A=@n17) {constB=0.9967})) {constB=0.25}
  n19 = Multiply(A=Frac(None=Add(A=@n16.B, B=@n18))) {constB=6}
  n20 = MaterialFunctionCall(Hue Shift Percentage=Divide(A=Floor(None=@n19)) {constB=6}, Texture=Constant3Vector() {constant=<rgba 1 0 0 0>}) {materialFunction=&HueShift}
  n21 = ScalarParameter() {defaultValue=15, parameterName=Brightness Bulbs}
  n22 = ScalarParameter() {defaultValue=4, parameterName=Brightness_Mobile}
  n23 = ShadingPathSwitch(
   Default=@n21
   Deferred=@n21
   Mobile=QualitySwitch(Default=@n22, Low=@n22, Medium=ScalarParameter() {defaultValue=4, parameterName=Brightness_Switch})
  )
  n24 = TextureSample() {samplerType=SAMPLERTYPE_Normal, texture=/Game/Environments/Asteria/Props/Holiday/KIT_Holiday_Garlands_A/Testures/T_Holiday_Garlands_A_A_N}
  n25 = SetMaterialAttributes!(
   MaterialAttributes
   Base Color=TextureSample() {texture=/Game/Environments/Asteria/Props/Holiday/KIT_Holiday_Garlands_A/Testures/T_Holiday_Garlands_A_A_D}.RGB
   Metallic=ScalarParameter() {defaultValue=1, parameterName=Metallic, sliderMax=1}
   Roughness=@n12.B
   Emissive Color=Multiply(A=Multiply(A=@n14, B=@n15), B=Multiply(A=@n20.Result, B=@n23))
   Shading Model=ShadingModel()
   Normal=@n24.RGB
   Specular
   Ambient Occlusion=@n12.B
  ) {attributeSetTypes [
    #1
    #2
    #4
    #5
    #10
    #6
    #3
    #11
   ]}
  n26 = GetMaterialAttributes(None=BlendMaterialAttributes(A=@n11, B=@n25, Alpha=NamedRerouteUsage())) {attributeGetTypes [#12, #4]}
  n27 = SetMaterialAttributes!(
   MaterialAttributes=@n26.MaterialAttributes
   World Position Offset=Add(
    A=@n26.World Position Offset
    B=MaterialFunctionCall() {materialFunction=&MF_FN_DestructibleDamage}.DamageAnimWPO
   )
  ) {attributeSetTypes [#12]}
  n28 = ScalarParameter() {defaultValue=0.3333, parameterName=MobileMinRoughness, sliderMax=1}
  ShadingPathSwitch(
   Default=@n27
   Deferred=@n27
   Forward=@n26.MaterialAttributes
   Mobile=SetMaterialAttributes!(MaterialAttributes=@n26.MaterialAttributes, Roughness=Clamp(None=@n26.Roughness, Min=@n28) {minDefault=0.4}) {attributeSetTypes [#4]}
  )
  VertexColor()
 ]
}
