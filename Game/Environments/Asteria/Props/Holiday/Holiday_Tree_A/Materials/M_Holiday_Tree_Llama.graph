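# Subsurface material for the llama on the Holiday Tree A prop, using llama body textures and flagged for instanced static meshes and Nanite.
# A fake interior glow raises a view-direction dot product to a power, using the direction from a camera-shifted local center to the pixel, distorted by its normal.
# The tinted glow is brightness-clamped on ES2 and ES3.1.
# MF_FN_DestructibleDamage supplies damage-animation WPO, switched to zero on forward and mobile paths.
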Material /Game/Environments/Asteria/Props/Holiday/Holiday_Tree_A/Materials/M_Holiday_Tree_Llama
graph {
 properties {
  bUsedWithInstancedStaticMeshes=true
  bUsedWithNanite=true
  pixelDepthOffsetMode=PDOM_Legacy
  shadingModel=MSM_Subsurface
  usedShadingModels=Subsurface
 }
 nodes [
  Constant() {r=1}
  n1 = WorldPosition() {worldPositionShaderOffset=WPT_ExcludeAllShaderOffsets}
  n2 = CameraVectorWS()
  n3 = Add(
   A=TransformPosition(None=VectorParameter() {group=Emissive, parameterName=EmissiveCenterPosition, defaultValue=<rgba 0 0 461.1604 1>}.RGB) {transformType=TRANSFORMPOSSOURCE_World}
   B=Multiply(A=@n2, B=ScalarParameter() {defaultValue=-201.2438, group=Emissive, parameterName=EmissiveCenterSoften})
  )
  n4 = PixelNormalWS()
  n5 = ScalarParameter() {defaultValue=26.3527, parameterName=EmissivePixelNormalDistort}
  n6 = Normalize(VectorInput=Add(A=Subtract(A=@n1.XYZ, B=@n3), B=Multiply(A=@n4, B=@n5)))
  n7 = Multiply(
   A=Power(Base=DotProduct(A=@n6, B=@n2), Exp=ScalarParameter() {defaultValue=64.0263, group=Emissive, parameterName=EmissiveFresnel_InterirorFalloff})
   B=VectorParameter() {group=Emissive, parameterName=EmissiveFresnel_CoreBrightness, defaultValue=<rgba 10 4.5485 0.2172 1>}.RGB
  )
  n8 = Clamp(None=@n7, Max=ScalarParameter() {defaultValue=5, parameterName=MobileMaxBrightness})
  FeatureLevelSwitch(Default=@n7, ES2=@n8, ES3_1=@n8, SM4_REMOVED=@n7)
  ScalarParameter() {parameterName=Subsurface Opacity, sliderMax=1}
  n9 = MaterialFunctionCall() {materialFunction=&MF_FN_DestructibleDamage}
  n10 = Constant()
  ShadingPathSwitch(Default=@n9.DamageAnimWPO, Deferred=@n9.DamageAnimWPO, Forward=@n10, Mobile=@n10)
  TextureSample() {texture=/Game/Environments/Props/Winter/Textures/T_lama_body_D}
  TextureSample() {samplerType=SAMPLERTYPE_LinearColor, texture=/Game/Environments/Props/Winter/Textures/T_lama_body_M}
  TextureSample() {samplerType=SAMPLERTYPE_Normal, texture=/Game/Environments/Props/Winter/Textures/T_lama_body_N}
  VectorParameter() {parameterName=Subsurface Color, defaultValue=<rgba 0.2604 0.0653 0.0263 0>}
 ]
}
